WidgetBlueprint /GiftBox/UI/Headers/CurrencyListViewItem
graphs [
 EventGraph {nodes [
   n1 = Event OnListItemObjectSet()
   > n2 = Cast To CurrencyListObject!(execute, Object=@n1.ListItemObject) -> (then, CastFailed, AsCurrency List Object:Currency List Object Object Reference) {is=title K2Node_DynamicCast}
   n3 = Get!(self:Currency List Object Object Reference=@n2) -> (Data_BundleTemplateId:Text, Data_Quantity:int, Data_QuestTemplateId:Text, Data_CurrencyTemplateId:Text)
   > SetQuantityAndType!(execute, self, InInt:int=@n3.Data_Quantity, To:Text=@n3.Data_CurrencyTemplateId) -> (then)
   n4 = "To String (Text)"(InText=@n3)
   > n5 = Branch(Condition=IsEmpty(InString=@n4))
   n6 = "To String (Text)"(InText=@n3.Data_QuestTemplateId)
   > n7 = Branch(Condition=IsEmpty(InString=@n6))
   n8 = "To String (Text)"(InText=MakeLiteralText(Value="LOCTEXT(\"all your hard work\")"))
   n9 = "To String (Text)"(InText=Get DisplayMessage() -> (DisplayMessage:Text))
   > SetText(
    self=Get CurrencyMessage() -> (CurrencyMessage:Common Rich Text Block Object Reference)
    InText="To Text (String)"(InString=Replace(SourceString=@n9, From="{Source}", To=@n8))
   )
   n10 = "To String (Text)"(InText=Get DisplayMessage())
   SetText(execute=@n7.else, self=Get CurrencyMessage(), InText="To Text (String)"(InString=Replace(SourceString=@n10, From="{Source}", To=@n6)))
   n11 = "To String (Text)"(InText=Get DisplayMessage())
   SetText(execute=@n5.else, self=Get CurrencyMessage(), InText="To Text (String)"(InString=Replace(SourceString=@n11, From="{Source}", To=@n4)))
  ]}
 SetQuantityAndType {nodes [
   n1 = SetQuantityAndType!() -> (then, InInt:int, To:Text) {is=title K2Node_FunctionEntry}
   n2 = "To String (Text)"(InText=@n1.To)
   n3 = Get CurrencyMessage()
   n4 = Replace(
    SourceString="To String (Text)"(InText=GetText!(self:Rich Text Block Object Reference=@n3) -> (ReturnValue:Text))
    From="{Quantity}"
    To="To String (Integer)"(InInt=@n1)
   )
   > Set DisplayMessage(DisplayMessage:Text="To Text (String)"(InString=Replace(SourceString=@n4, From="{Type}", To=@n2))) -> (Output_Get:Text)
  ]}
]
variables [DisplayMessage {type=string, default=""}]
widgets {
 info {inheritedWidgetCount=0, parentClass=&UserWidget, rootWidgetClass=&Border, widgetCount=2}
 widgets [
  Border_0 {
   widgetClassPath=&Border
   widget=Border_0
   properties {
    background {
     bIsDynamicallyLoaded=false
     drawAs=NoDrawType
     imageType=NoImage
     mirroring=NoMirror
     resourceName=None
     resourceObject=null
     tiling=NoTile
     imageSize=<32 32>
     margin=<ltrb 0 0 0 0>
     outlineSettings {
      bUseBrushTransparency=false
      roundingType=HalfHeightRadius
      width=0
      color {colorUseRule=UseColor_Specified, specifiedColor=<rgba 0 0 0 0>}
      cornerRadii=<0 0 0 0>
     }
     tintColor {colorUseRule=UseColor_Specified, specifiedColor=<rgba 1 1 1 1>}
     uVRegion {bIsValid=false, max=<0 0>, min=<0 0>}
    }
    contentColorAndOpacity=<rgba 0.6038 1 1 1>
    padding=<ltrb 0 0 0 0>
   }
  }
  CurrencyMessage {
   widgetClassPath=&CommonRichTextBlock
   bIsVariable=true
   parent=Border_0
   widget=CurrencyMessage
   properties {
    autoWrapText=true
    bOverrideDefaultStyle=true
    bTintInlineIcon=true
    defaultTextStyleOverrideClass=/FortUILibrary/Styles/Typography/_TextStyle_Body1._TextStyle_Body1_C
    inlineIconDisplayMode=TextOnly
    mobileTextBlockScale=1.3
    text="You earned {Quantity} {Type} from {Source}!"
    defaultTextStyleOverride {
     overflowPolicy=Clip
     transformPolicy=None
     colorAndOpacity {colorUseRule=UseColor_Specified, specifiedColor=<rgba 1 1 1 1>}
     font {
      bForceMonospaced=false
      bMaterialIsStencil=false
      fontMaterial=null
      fontObject=&NotoSans
      letterSpacing=0
      monospacedWidth=1
      size=24
      skewAmount=0
      typefaceFontName=NotoSans-Regular
      outlineSettings {
       bApplyOutlineToDropShadows=false
       bMiteredCorners=false
       bSeparateFillAlpha=false
       outlineMaterial=null
       outlineSize=0
       outlineColor=<rgba 0 0 0 1>
      }
     }
     highlightColor {colorUseRule=UseColor_Specified, specifiedColor=<rgba 0 0 0 1>}
     highlightShape {
      bIsDynamicallyLoaded=false
      drawAs=Image
      imageType=NoImage
      mirroring=NoMirror
      resourceName=None
      resourceObject=null
      tiling=NoTile
      imageSize=<32 32>
      margin=<ltrb 0 0 0 0>
      outlineSettings {
       bUseBrushTransparency=false
       roundingType=HalfHeightRadius
       width=0
       color {colorUseRule=UseColor_Specified, specifiedColor=<rgba 0 0 0 0>}
       cornerRadii=<0 0 0 0>
      }
      tintColor {colorUseRule=UseColor_Specified, specifiedColor=<rgba 1 1 1 1>}
      uVRegion {bIsValid=false, max=<0 0>, min=<0 0>}
     }
     selectedBackgroundColor {colorUseRule=UseColor_Foreground, specifiedColor=<rgba 1 0 1 1>}
     shadowColorAndOpacity=<rgba 0 0 0 1>
     shadowOffset=<0 0>
     strikeBrush {
      bIsDynamicallyLoaded=false
      drawAs=Image
      imageType=NoImage
      mirroring=NoMirror
      resourceName=None
      resourceObject=null
      tiling=NoTile
      imageSize=<32 32>
      margin=<ltrb 0 0 0 0>
      outlineSettings {
       bUseBrushTransparency=false
       roundingType=HalfHeightRadius
       width=0
       color {colorUseRule=UseColor_Specified, specifiedColor=<rgba 0 0 0 0>}
       cornerRadii=<0 0 0 0>
      }
      tintColor {colorUseRule=UseColor_Specified, specifiedColor=<rgba 1 1 1 1>}
      uVRegion {bIsValid=false, max=<0 0>, min=<0 0>}
     }
     underlineBrush {
      bIsDynamicallyLoaded=false
      drawAs=Image
      imageType=NoImage
      mirroring=NoMirror
      resourceName=None
      resourceObject=null
      tiling=NoTile
      imageSize=<32 32>
      margin=<ltrb 0 0 0 0>
      outlineSettings {
       bUseBrushTransparency=false
       roundingType=HalfHeightRadius
       width=0
       color {colorUseRule=UseColor_Specified, specifiedColor=<rgba 0 0 0 0>}
       cornerRadii=<0 0 0 0>
      }
      tintColor {colorUseRule=UseColor_Specified, specifiedColor=<rgba 1 1 1 1>}
      uVRegion {bIsValid=false, max=<0 0>, min=<0 0>}
     }
    }
   }
   slot {class=BorderSlot, properties {padding=<ltrb 0 0 0 0>}}
  }
 ]
}
