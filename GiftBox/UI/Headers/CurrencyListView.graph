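# CurrencyListView is a UserWidget that shows currency entries in the GiftBox header UI through a single CommonListView root.
# AddDataToList wraps each Currency Data Structure in a new CurrencyListObject and adds it to the list as an item.
# The list displays items with CurrencyListViewItem entry widgets.
# A RunClearListView event clears all list items.
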WidgetBlueprint /GiftBox/UI/Headers/CurrencyListView
graphs [
 AddDataToList {nodes [
   n1 = AddDataToList!() -> (then, CurrencyData:Currency Data Structure) {is=title K2Node_FunctionEntry}
   > n2 = Construct Object from Class(Class=&/CurrencyListObject.CurrencyListObject_C) -> (ReturnValue:Currency List Object Object Reference)
   > Set Data!(execute, Data:Currency Data Structure=@n1, self:Currency List Object Object Reference=@n2) -> (then, Output_Get:Currency Data Structure)
   > AddItem!(execute, self:List View Object Reference=Get DataListView() -> (DataListView:Common List View Object Reference), Item:Object Reference=@n2) -> (then)
  ]}
 EventGraph {nodes [
   RunClearListView() {is=title K2Node_CustomEvent}
   > ClearListItems(self=Get DataListView())
  ]}
]
variables []
widgets {
 info {inheritedWidgetCount=0, parentClass=&UserWidget, rootWidgetClass=&CommonListView, widgetCount=1}
 widgets [DataListView {
   widgetClassPath=&CommonListView
   bIsVariable=true
   widget=DataListView
   properties {entryWidgetClass=&/CurrencyListViewItem.CurrencyListViewItem_C}
  }]
}
