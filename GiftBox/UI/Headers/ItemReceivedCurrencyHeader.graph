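# Header sub-widget for the item-received gift box screen that lists the currencies received.
# The ActiveFromPresentation event clears CurrencyListView_82, then loops over the passed currency data array, adding each entry to the list.
# Above the list, a title switcher holds "Congratulations!" and Battle Pass tier-up texts, alongside a collapsed account widget.
# No graph here switches the title or uses AlreadyActivated, and ActivateWidgetFromPresentationScreen has only an entry node.
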WidgetBlueprint /GiftBox/UI/Headers/ItemReceivedCurrencyHeader
graphs [
 ActivateWidgetFromPresentationScreen {nodes [ActivateWidgetFromPresentationScreen!() -> (then, GiftBox:Fort Gift Box Item Object Reference, CurrencyArray:Array of Currency Data Structures) {is=title K2Node_FunctionEntry}]}
 EventGraph {nodes [
   n1 = ActiveFromPresentation!() -> (OutputDelegate, then, CurrencyDataArray_GiftBoxCurrencyData:Array of Currency Data Structures) {is=title K2Node_CustomEvent}
   n2 = Get CurrencyListView_82() -> (CurrencyListView_82:Currency List View Object Reference)
   > RunClearListView(self:Currency List View Object Reference=@n2)
   > n3 = For Each Loop(Array:Array of Currency Data Structures=@n1.CurrencyDataArray_GiftBoxCurrencyData) -> (Array Element:Currency Data Structure)
   > AddDataToList!(execute, self:Currency List View Object Reference=@n2, CurrencyData:Currency Data Structure=@n3) -> (then)
  ]}
]
variables [
 ActivateWidgetFromPresentationScreen {}
 AlreadyActivated {type=boolean}
]
widgets {
 info {
  inheritedWidgetCount=0
  parentClass=/Game/UI/ItemReceived/ItemReceivedHeaders/ItemReceivedHeaderSubWidgetBase.ItemReceivedHeaderSubWidgetBase_C
  rootWidgetClass=&HorizontalBox
  widgetCount=9
 }
 widgets [
  HorizontalBox_0 {widgetClassPath=&HorizontalBox, widget=HorizontalBox_0}
  VerticalBox_2 {
   widgetClassPath=&VerticalBox
   parent=HorizontalBox_0
   widget=VerticalBox_2
   slot {class=HorizontalBoxSlot, properties {verticalAlignment=VAlign_Center, size {sizeRule=Fill, value=1}}}
  }
  BorderItemNAme {
   widgetClassPath=&Border
   parent=VerticalBox_2
   widget=BorderItemNAme
   properties {
    horizontalAlignment=HAlign_Left
    background {
     bIsDynamicallyLoaded=false
     drawAs=NoDrawType
     imageType=NoImage
     mirroring=NoMirror
     resourceName=None
     resourceObject=null
     tiling=NoTile
     imageSize=<32 32>
     margin=<ltrb 0 0 0 0>
     outlineSettings {
      bUseBrushTransparency=false
      roundingType=HalfHeightRadius
      width=0
      color {colorUseRule=UseColor_Specified, specifiedColor=<rgba 0 0 0 0>}
      cornerRadii=<0 0 0 0>
     }
     tintColor {colorUseRule=UseColor_Specified, specifiedColor=<rgba 1 1 1 1>}
     uVRegion {bIsValid=false, max=<0 0>, min=<0 0>}
    }
    padding=<ltrb 0 0 0 0>
   }
   slot {class=VerticalBoxSlot, properties {size {sizeRule=Fill, value=1}}}
  }
  CommonWidgetSwitcher_Title {
   widgetClassPath=&CommonWidgetSwitcherLegacy
   bIsVariable=true
   parent=BorderItemNAme
   widget=CommonWidgetSwitcher_Title
   slot {class=BorderSlot, properties {horizontalAlignment=HAlign_Left, padding=<ltrb 0 0 0 0>}}
  }
  BattlepassUpgraded {
   widgetClassPath=&CommonTextBlock
   bIsVariable=true
   parent=CommonWidgetSwitcher_Title
   widget=BattlepassUpgraded
   properties {
    justification=Center
    text="Congratulations!"
    textTransformPolicy=ToUpper
    font {
     bForceMonospaced=false
     bMaterialIsStencil=false
     fontMaterial=null
     fontObject=&Burbank
     letterSpacing=0
     monospacedWidth=1
     size=42
     skewAmount=0
     typefaceFontName=BurbankRegular-Black
     outlineSettings {
      bApplyOutlineToDropShadows=false
      bMiteredCorners=false
      bSeparateFillAlpha=false
      outlineMaterial=null
      outlineSize=0
      outlineColor=<rgba 0 0 0 1>
     }
    }
    shadowOffset=<0 0>
    strikeBrush {
     bIsDynamicallyLoaded=false
     drawAs=NoDrawType
     imageType=NoImage
     mirroring=NoMirror
     resourceName=None
     resourceObject=null
     tiling=NoTile
     imageSize=<32 32>
     margin=<ltrb 0 0 0 0>
     outlineSettings {
      bUseBrushTransparency=false
      roundingType=HalfHeightRadius
      width=0
      color {colorUseRule=UseColor_Specified, specifiedColor=<rgba 0 0 0 0>}
      cornerRadii=<0 0 0 0>
     }
     tintColor {colorUseRule=UseColor_Specified, specifiedColor=<rgba 1 1 1 1>}
     uVRegion {bIsValid=false, max=<0 0>, min=<0 0>}
    }
   }
   slot {class=WidgetSwitcherSlot, properties {horizontalAlignment=HAlign_Center}}
  }
  BattlepassTier {
   widgetClassPath=&CommonTextBlock
   bIsVariable=true
   parent=CommonWidgetSwitcher_Title
   widget=BattlepassTier
   properties {
    justification=Center
    text="Battle Pass TIER UP!"
    textTransformPolicy=ToUpper
    font {
     bForceMonospaced=false
     bMaterialIsStencil=false
     fontMaterial=null
     fontObject=&Burbank
     letterSpacing=0
     monospacedWidth=1
     size=42
     skewAmount=0
     typefaceFontName=BurbankRegular-Black
     outlineSettings {
      bApplyOutlineToDropShadows=false
      bMiteredCorners=false
      bSeparateFillAlpha=false
      outlineMaterial=null
      outlineSize=0
      outlineColor=<rgba 0 0 0 1>
     }
    }
    shadowOffset=<0 0>
    strikeBrush {
     bIsDynamicallyLoaded=false
     drawAs=NoDrawType
     imageType=NoImage
     mirroring=NoMirror
     resourceName=None
     resourceObject=null
     tiling=NoTile
     imageSize=<32 32>
     margin=<ltrb 0 0 0 0>
     outlineSettings {
      bUseBrushTransparency=false
      roundingType=HalfHeightRadius
      width=0
      color {colorUseRule=UseColor_Specified, specifiedColor=<rgba 0 0 0 0>}
      cornerRadii=<0 0 0 0>
     }
     tintColor {colorUseRule=UseColor_Specified, specifiedColor=<rgba 1 1 1 1>}
     uVRegion {bIsValid=false, max=<0 0>, min=<0 0>}
    }
   }
   slot {class=WidgetSwitcherSlot}
  }
  VerticalBox_0 {
   widgetClassPath=&VerticalBox
   parent=VerticalBox_2
   widget=VerticalBox_0
   slot {class=VerticalBoxSlot, properties {padding=<ltrb 0 8 0 0>}}
  }
  ItemReceivedAccountWidget {
   widgetClassPath=/Game/UI/ItemReceived/ItemReceivedHeaders/ItemReceivedAccountWidget
   bIsVariable=true
   parent=VerticalBox_0
   widget=ItemReceivedAccountWidget
   properties {visibility=Collapsed}
   slot {class=VerticalBoxSlot}
  }
  CurrencyListView_82 {
   widgetClassPath=&/CurrencyListView
   bIsVariable=true
   parent=VerticalBox_0
   widget=CurrencyListView_82
   slot {class=VerticalBoxSlot, properties {size {sizeRule=Fill, value=1}}}
  }
 ]
}
